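# Unlit, two-sided translucent graffiti trail material for the female Cadet pickaxe, usable on Niagara ribbons.
# A vertically panned blue channel of the graffiti texture blends outline and core colors, scaled by dynamic EmissiveAmount or red-channel detail with camera depth fade.
# Dynamic parameters flip and pan the detail sample.
# Its alpha combines with inverted green-channel masks and particle alpha; MasterAlpha is declared but unread.
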Material /BRCosmetics/Weapons/FORT_Melee/Pickaxe_Cadet_Female/FX/M_Pickaxe_Cadet_Graffiti_Trail
graph {
 properties {
  bUsedWithNiagaraRibbons=true
  blendMode=BLEND_Translucent
  pixelDepthOffsetMode=PDOM_Legacy
  shadingModel=MSM_Unlit
  twoSided=true
  usedShadingModels=Unlit
 }
 nodes [
  n1 = TextureCoordinate() {coordinateIndex=1}
  n2 = Time()
  n3 = LinearInterpolate(
   A=VectorParameter() {parameterName=Color_Outline, defaultValue=<rgba 0.724 0.0228 0.1693 1>}.RGB
   B=VectorParameter() {parameterName=Color_Core, defaultValue=<rgba 0.0856 0.5711 0.724 1>}.RGB
   Alpha=TextureSampleParameter2D(UVs=Panner(Coordinate=@n1, Time=@n2) {speedY=0.3}) {parameterName=TrailTexture_BChannel, texture=&/T_Pickaxe_Pickaxe_Graffiti}.B
  )
  MaterialFunctionCall(Emissive=Multiply(A=DynamicParameter() {paramNames ["Flip Trail (-1 or 1)", EmissiveAmount, PanSpeed, MasterAlpha]}.EmissiveAmount, B=@n3)) {materialFunction=&MF_FX_EmissiveAdjustments}
  n4 = TextureCoordinate()
  n5 = ComponentMask(None=@n4) {r=true}
  n6 = Constant()
  n7 = Constant() {r=5}
  n8 = OneMinus(None=@n5)
  Multiply(
   A=Saturate(None=MaterialFunctionCall(Input=@n5, Target Low=@n6, Target High=@n7) {materialFunction=&RemapValueRange}.Result)
   B=Saturate(None=MaterialFunctionCall(Input=@n8, Target Low=@n6, Target High=@n7) {materialFunction=&RemapValueRange}.Result)
  ) {constA=1}
  n9 = Multiply(A=ComponentMask(None=@n1) {r=true}) {constB=1.8}
  n10 = DynamicParameter() {paramNames ["Flip Trail (-1 or 1)", EmissiveAmount, PanSpeed, MasterAlpha]}
  n11 = Multiply(A=ComponentMask(None=@n1) {g=true}, B=@n10."Flip Trail (-1 or 1)")
  n12 = Constant()
  n13 = TextureSampleParameter2D(UVs=Panner(Coordinate=AppendVector(A=@n9, B=@n11), Time=Time(), Speed=AppendVector(A=@n10.PanSpeed, B=@n12))) {parameterName=TrailTexture_BChannel, texture=&T_Pickaxe_Pickaxe_Graffiti}
  Multiply(
   A=Multiply(A=Add(A=@n13.R), B=@n3)
   B=MaterialFunctionCall(
    Fade Length=ScalarParameter() {defaultValue=30, parameterName=Fade Length}
    Fade Offset=ScalarParameter() {defaultValue=24, parameterName=Fade Offset}
   ) {materialFunction=&CameraDepthFade}.Result
  )
  n14 = Add(A=ComponentMask(None=@n4) {g=true}) {constB=0.3}
  n15 = OneMinus(None=TextureSampleParameter2D(UVs=AppendVector(A=@n8, B=@n14)) {parameterName=TrailTexture_BChannel, texture=&T_Pickaxe_Pickaxe_Graffiti}.G)
  n16 = OneMinus(None=TextureSampleParameter2D(UVs=TextureCoordinate() {coordinateIndex=1}) {parameterName=TrailTexture_BChannel, texture=&T_Pickaxe_Pickaxe_Graffiti}.G)
  Multiply(A=ParticleColor().A, B=Multiply(A=@n13.A, B=Multiply(A=@n15, B=@n16) {constA=1}))
 ]
}
